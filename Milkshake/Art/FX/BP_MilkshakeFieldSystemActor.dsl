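; Field system actor that shakes trees, foliage and similar objects in proportion to the strength of its owning earthquake.
; On BeginPlay, where cosmetic events can run, it starts a looping timer at a configurable interval.
; Each tick of that timer reads the owner earthquake's normalized strength and scales it by a maximum to set the field strength, so shake rises and falls with the quake.
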(event EventBeginPlay
  ;[0.04624 0.4427 0] BeginPlay: Start a timer to update the strength of the field system.
  (|CallParentFunction)
  (Networking|CanExecuteCosmeticEvents
    (:True
      (Utilities|Time|SetTimerbyEvent (EventDispatchers|CreateEvent :!event Custom|UpdateStrength) (Variables|Default|GetUpdateStrengthInterval) true))))

(event Custom|UpdateStrength
  ;[0 0.2083 0.2034] UpdateStrength: The normalized earthquake strength is used to calculate the strength of the field system actor.
  ; As the earthquake gets stronger, there is more shake to the trees/foliage etc and vice versa.
  (Networking|CanExecuteCosmeticEvents
    (:True
      (bind _ownerearthquake (Variables|Default|GetOwnerEarthquake)
        (:then
          (bind _value (General|GetNormalizedEarthquakeStrength _ownerearthquake))
          (MilkshakeFieldSystemActor|SetStrength (* _value (Variables|Default|GetMaxStrength))))
        (:else)))))

(fn ConstructionScript)
